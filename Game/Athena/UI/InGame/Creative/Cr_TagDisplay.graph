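# Creative UI widget that displays a wrapping row of tag labels from island description tags.
# Its only widget is a Dynamic Entry Box that spawns Cr_ItemTag entries.
# PopulateTags clears the box, loops over the given tag structures, and creates an entry for each tag with non-empty display text.
# Each new entry is cast to Cr_ItemTag and updated with that text.
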WidgetBlueprint /Game/Athena/UI/InGame/Creative/Cr_TagDisplay
graphs [
 EventGraph {links [], nodes []}
 PopulateTags {nodes [
   n1 = PopulateTags!() -> (then, inTags:Array of Creative Island Description Tag Structures) {is=title K2Node_FunctionEntry}
   > Reset!(
    execute
    self:Dynamic Entry Box Object Reference=Get EntryBox_ActivityTags() -> (EntryBox_ActivityTags:Dynamic Entry Box Object Reference)
    bDeleteWidgets:bool=true
   ) -> (then)
   > n2 = For Each Loop(Array:Array of Creative Island Description Tag Structures=@n1) -> (Array Element:Creative Island Description Tag Structure)
   n3 = Break Creative Island Description Tag!(CreativeIslandDescriptionTag:"Creative Island Description Tag Structure (by ref)"=@n2) -> (Name:String, DisplayText:Text) {is=title K2Node_BreakStruct}
   > n4 = Branch(Condition=TextIsEmpty(InText=@n3.DisplayText))
   n5 = Create Entry!(execute=@n4.else, self:Dynamic Entry Box Object Reference=Get EntryBox_ActivityTags()) -> (then, ReturnValue:User Widget Object Reference)
   > n6 = Cast To Cr_ItemTag!(execute, Object=@n5) -> (then, CastFailed, AsCr Item Tag:Cr Item Tag Object Reference) {is=title K2Node_DynamicCast}
   > Update!(execute, self:Cr Item Tag Object Reference=@n6, inTagText:Text=@n3.DisplayText) -> (then)
  ]}
]
variables []
widgets {
 info {inheritedWidgetCount=0, parentClass=&CommonUserWidget, rootWidgetClass=&DynamicEntryBox, widgetCount=1}
 widgets [EntryBox_ActivityTags {
   widgetClassPath=&DynamicEntryBox
   bIsVariable=true
   widget=EntryBox_ActivityTags
   properties {entryBoxType=Wrap, entryWidgetClass=/Game/UI/Inventory/Cr_ItemTag.Cr_ItemTag_C, numDesignerPreviewEntries=4, entrySpacing=<4 6>}
  }]
}
