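# Creative affiliate panel widget that wraps AffiliateKeyModal to show support-a-creator (SAC) details in Creative.
# InitFromObject pushes it as a modal, hides the website button and fills name, source and editability from the object's BPi_Creative_SACActor interface.
# Activation plays a sound and binds modal events, relaying success; the close button or popup dismissal pops the panel.
# Deactivation unlocks cursor mode, restores camera mode and broadcasts OnPanelClose.
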WidgetBlueprint /Game/Athena/UI/InGame/Creative/CreativeAffiliatePanel
graphs [
 Close {nodes [
   Close() {is=title K2Node_FunctionEntry}
   > PopActivatablePanelInModalLayer!(execute, Panel:Common Activatable Panel Legacy Object Reference=Self-Reference()) -> (then)
  ]}
 EventGraph {nodes [
   n1 = Get FortHUDContext!() -> (ReturnValue:Fort HUDContext Object Reference) {is=title K2Node_GetBlueprintContextEx}
   EnterCursorMode!(execute, self:Fort HUDContext Object Reference=@n1, ActionName:Name=None, CursorModeWidget:User Widget Object Reference=Self-Reference()) -> (then)
   > n2 = SetCursorModeLocked!(execute, self:Fort HUDContext Object Reference=@n1, bLocked:bool=true) -> (then)
   n3 = PlaySound2D(execute=[@n2.then, @n14.then, @n16.else, @n16.then], Sound=Get OpenMenuSound() -> (OpenMenuSound:Sound Base Object Reference))
   n4 = OnCloseButtonClicked!() -> (OutputDelegate, then, Button:Common Button Base Object Reference) {is=title K2Node_CustomEvent}
   n5 = Get AffiliateKeyModal() -> (AffiliateKeyModal:Affiliate Key Modal Object Reference)
   n6 = Bind Event to On Clicked(
    execute=@n3.then
    self:Common Button Base Object Reference=Get CloseButton!(self:Affiliate Key Modal Object Reference=@n5) -> (CloseButton:WBP UIKit Button Regular Object Reference)
    Delegate=@n4
   ) {is=title K2Node_AddDelegate}
   n7 = PopupClosed() {is=title K2Node_CustomEvent}
   n8 = Bind Event to On Popup Closed(execute=@n6.then, self:Affiliate Key Modal Object Reference=@n5, Delegate=@n7) {is=title K2Node_AddDelegate}
   n9 = On Success() {is=title K2Node_CustomEvent}
   Bind Event to On Success(execute=@n8.then, self:Affiliate Key Modal Object Reference=@n5, Delegate=@n9) {is=title K2Node_AddDelegate}
   n10 = Event InitFromObject!() -> (OutputDelegate, then, InitObject:Object Reference) {is=title K2Node_Event}
   > TryPushWidgetAsModalPanel!(execute, PushWidget:User Widget Object Reference=Self-Reference()) -> (then)
   n11 = Get AffiliateKeyModal()
   > ActivateWidget(self=@n11)
   > SetVisibility(
    self=Get LaunchWebsite!(self:Affiliate Key Modal Object Reference=@n11) -> (LaunchWebsite:WBP UIKit Button Regular Object Reference)
    InVisibility=Collapsed
   )
   > n12 = Cast To BPi_Creative_SACActor!(execute, Object=@n10.InitObject) -> (then, CastFailed, AsBPi Creative SACActor:BPi Creative SACActor Interface) {is=title K2Node_DynamicCast}
   > n13 = GetSACDetails!(execute, self:BPi Creative SACActor Interface=@n12) -> (then, Affiliate Name:Text, Source Name:String, CanEditAffiliate:bool)
   > SetAffiliateName!(execute, self, AffiliateName:Text=@n13, Can Edit:bool=@n13.CanEditAffiliate) -> (then)
   > SetSourceName!(execute, self, NewSourceName:String=@n13.Source Name) -> (then)
   Event On Activated()
   > n14 = Parent: On Activated() {is=title K2Node_CallParentFunction}
   Event On Deactivated()
   n15 = Get FortHUDContext()
   > SetCursorModeLocked(self=@n15, bLocked=false)
   > EnterCameraMode(self:Fort HUDContext Object Reference=@n15)
   > Parent: On Deactivated() {is=title K2Node_CallParentFunction}
   > PlaySound2D(Sound=Get CloseMenuSound() -> (CloseMenuSound:Sound Base Object Reference))
   > Call On Panel Close!(execute:Exec, self:Creative Affiliate Panel Object Reference) -> (then:Exec) {is=title K2Node_CallDelegate}
   n16 = Branch(Condition=IsInCursorMode!(self:Fort HUDContext Object Reference=@n1) -> (ReturnValue:bool))
   Call On Success!(execute:Exec=@n9.then, self:Creative Affiliate Panel Object Reference) -> (then:Exec) {is=title K2Node_CallDelegate}
   Close(execute=@n4.then) {is=title K2Node_CallFunction}
   Close(execute=@n18.then)
   n17 = Get AffiliateKeyModal()
   Unbind all Events from On Popup Closed!(execute:Exec=@n7.then, self:Affiliate Key Modal Object Reference=@n17) -> (then:Exec) {is=title K2Node_ClearDelegate}
   > n18 = Unbind all Events from On Success!(execute:Exec, self:Affiliate Key Modal Object Reference=@n17) -> (then:Exec) {is=title K2Node_ClearDelegate}
  ]}
 OnPanelClose {nodes [OnPanelClose() {is=title K2Node_FunctionEntry}]}
 OnSuccess {nodes [OnSuccess() {is=title K2Node_FunctionEntry}]}
 SetAffiliateName {nodes [
   n1 = SetAffiliateName!() -> (then, AffiliateName:Text, Can Edit:bool=true) {is=title K2Node_FunctionEntry}
   n2 = Get AffiliateKeyModal()
   > Is Valid(InputObject=@n2)
   n3 = Get EditableText!(self:Affiliate Key Modal Object Reference=@n2) -> (EditableText:UIKit Editable Text Object Reference)
   > "SetText (Editable Text)"!(execute, self:Editable Text Object Reference=@n3, InText:Text=@n1) -> (then)
   > n4 = Branch(Condition=IsAffiliateKeyReadOnly!(self:Affiliate Key Modal Object Reference=@n2) -> (bIsAffiliateKeyReadOnly:bool))
   "SetIsReadOnly (Editable Text)"!(execute=@n4.else, self:Editable Text Object Reference=@n3, InbIsReadyOnly:bool=NOT Boolean(A=@n1.Can Edit)) -> (then)
  ]}
 SetSourceName {nodes [
   n1 = SetSourceName!() -> (then, NewSourceName:String) {is=title K2Node_FunctionEntry}
   > Set SourceName!(execute, SourceName:String=@n1, self:Fort Affilate Modal Object Reference=Get AffiliateKeyModal()) -> (then, Output_Get:String)
  ]}
]
variables [
 CloseMenuSound {type=SoundBase, default=/Game/Sounds/Fort_UI_Sounds/FrontEnd/New_FrontEnd/Fortnite_UI_MissionObj_Cue}
 OnPanelClose {}
 OnSuccess {}
 OpenMenuSound {type=SoundBase, default=/Game/Sounds/Fort_UI_Sounds/FrontEnd/New_FrontEnd/Fortnite_UI_StartButton_03_Cue}
]
widgets {
 info {inheritedWidgetCount=0, parentClass=/Script/FortniteUI.FortCreativeAffiliatePanel, rootWidgetClass=&Border, widgetCount=2}
 widgets [
  Border_0 {
   widgetClassPath=&Border
   widget=Border_0
   properties {
    background {
     bIsDynamicallyLoaded=false
     drawAs=NoDrawType
     imageType=NoImage
     mirroring=NoMirror
     resourceName=None
     resourceObject=null
     tiling=NoTile
     imageSize=<32 32>
     margin=<ltrb 0 0 0 0>
     outlineSettings {
      bUseBrushTransparency=false
      roundingType=HalfHeightRadius
      width=0
      color {colorUseRule=UseColor_Specified, specifiedColor=<rgba 0 0 0 0>}
      cornerRadii=<0 0 0 0>
     }
     tintColor {colorUseRule=UseColor_Specified, specifiedColor=<rgba 1 1 1 1>}
     uVRegion {bIsValid=false, max=<0 0>, min=<0 0>}
    }
    padding=<ltrb 0 0 0 0>
   }
  }
  AffiliateKeyModal {
   widgetClassPath=/Game/Athena/UI/Frontend/AffiliateKeyModal
   bIsVariable=true
   parent=Border_0
   widget=AffiliateKeyModal
   properties {bIsAffiliateKeyReadOnly=true}
   slot {class=BorderSlot, properties {padding=<ltrb 0 0 0 0>}}
  }
 ]
}
