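# Toggleable CommonButtonBase button for Creative in-game UI to express support, with icon, text label and pulse image.
# Construct events apply configurable text, materials, icon size and optional 180-degree icon rotation.
# Hover and selection events play widget animations, some reversed, driving material parameters, text color and a scaling pulse; disabling dims icon and text.
# SetIsFavorite presets the icon's active state and flags the next selection to skip its animation.
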WidgetBlueprint /Game/Athena/UI/InGame/Creative/WBP_ExpressYourSupportButton
animations [
 HoveringSelected {
  binding_count=3
  length_seconds=0.2
  object=&self:HoveringSelected
  sequence {bindings [
    Image_Icon {
     id=#1
     possessed_class=Image
     tracks [{
       type=MovieSceneWidgetMaterialTrack
       sections [{
         type=MovieSceneParameterSection
         channels [
          bIsInactive {type=MovieSceneScriptingFloatChannel, const [0,1]}
          bOnHover_RGBA {type=MovieSceneScriptingFloatChannel, keys [[0,0],[4,1]]}
         ]
        }]
      }]
    }
    ButtonText {
     id=#2
     possessed_class=CommonTextBlock
     tracks [{
       type=MovieSceneColorTrack
       display_name=Color and Opacity
       sections [{
         type=MovieSceneColorSection
         channels [
          Color.R {type=MovieSceneScriptingFloatChannel, keys [[0,1],[4,0]]}
          Color.G {type=MovieSceneScriptingFloatChannel, keys [[0,1],[4,0]]}
          Color.B {type=MovieSceneScriptingFloatChannel, keys [[0,1],[4,0]]}
          Color.A {type=MovieSceneScriptingFloatChannel, const [1,2]}
         ]
        }]
      }]
    }
    Background {
     id=#3
     possessed_class=Image
     tracks [{
       type=MovieSceneWidgetMaterialTrack
       sections [{
         type=MovieSceneParameterSection
         end_frame=4
         start_frame=0
         channels [AnimateHighlight {type=MovieSceneScriptingFloatChannel, keys [[0,0],[4,1]]}]
        }]
      }]
    }
   ]}
 }
 HoveringDeselected {
  binding_count=3
  length_seconds=0.2
  object=&self:HoveringDeselected
  sequence {bindings [
    ButtonText {
     id=#4
     possessed_class=CommonTextBlock
     tracks [{
       type=MovieSceneColorTrack
       display_name=Color and Opacity
       sections [{
         type=MovieSceneColorSection
         channels [
          Color.R {type=MovieSceneScriptingFloatChannel, keys [[0,1],[4,0]]}
          Color.G {type=MovieSceneScriptingFloatChannel, keys [[0,1],[4,0]]}
          Color.B {type=MovieSceneScriptingFloatChannel, keys [[0,1],[4,0]]}
          Color.A {type=MovieSceneScriptingFloatChannel, const [1,2]}
         ]
        }]
      }]
    }
    Background {
     id=#5
     possessed_class=Image
     tracks [{
       type=MovieSceneWidgetMaterialTrack
       sections [{
         type=MovieSceneParameterSection
         end_frame=4
         start_frame=0
         channels [AnimateHighlight {type=MovieSceneScriptingFloatChannel, keys [[0,0],[4,1]]}]
        }]
      }]
    }
    Image_Icon {
     id=#6
     possessed_class=Image
     tracks [{
       type=MovieSceneWidgetMaterialTrack
       sections [{type=MovieSceneParameterSection, channels [bOnHover_RGBA {type=MovieSceneScriptingFloatChannel, keys [[0,0],[3,1]]}]}]
      }]
    }
   ]}
 }
 HoveredSelecting {
  binding_count=3
  length_seconds=0.2
  object=&self:HoveredSelecting
  sequence {bindings [
    Image_Icon {
     id=#7
     possessed_class=Image
     tracks [{
       type=MovieSceneWidgetMaterialTrack
       sections [{
         type=MovieSceneParameterSection
         channels [
          bIsInactive {type=MovieSceneScriptingFloatChannel, keys [[0,1],[4,0]]}
          bOnHover_RGBA {type=MovieSceneScriptingFloatChannel, keys [[0,1],[4,0]]}
         ]
        }]
      }]
    }
    Background {
     id=#8
     possessed_class=Image
     tracks [{
       type=MovieSceneWidgetMaterialTrack
       sections [{type=MovieSceneParameterSection, end_frame=4, start_frame=0, channels [AnimateHighlight {type=MovieSceneScriptingFloatChannel, const [1,2]}]}]
      }]
    }
    ButtonText {
     id=#9
     possessed_class=CommonTextBlock
     tracks [{
       type=MovieSceneColorTrack
       display_name=Color and Opacity
       sections [{
         type=MovieSceneColorSection
         channels [
          Color.R {type=MovieSceneScriptingFloatChannel, const [0,2]}
          Color.G {type=MovieSceneScriptingFloatChannel, const [0,2]}
          Color.B {type=MovieSceneScriptingFloatChannel, const [0,2]}
          Color.A {type=MovieSceneScriptingFloatChannel, const [1,2]}
         ]
        }]
      }]
    }
   ]}
 }
 SelectedPulse {
  binding_count=2
  length_seconds=0.48
  object=&self:SelectedPulse
  sequence {bindings [
    Image_Pulse {
     id=#10
     possessed_class=Image
     tracks [
      {
       type=MovieSceneFloatTrack
       display_name=Render Opacity
       sections [{type=MovieSceneFloatSection, channels [None {type=MovieSceneScriptingFloatChannel, keys [[1,0],[4,1],[5.6,1],[9.6,0]]}]}]
      }
      {
       type=MovieScene2DTransformTrack
       sections [{
         type=MovieScene2DTransformSection
         channels [
          Scale.X {type=MovieSceneScriptingFloatChannel, keys [[2,1],[7.6,2.5]]}
          Scale.Y {type=MovieSceneScriptingFloatChannel, keys [[2,1],[7.6,2.5]]}
         ]
        }]
      }
     ]
    }
    Image_Icon {
     id=#11
     possessed_class=Image
     tracks [{
       type=MovieSceneWidgetMaterialTrack
       sections [{
         type=MovieSceneParameterSection
         channels [
          bIsInactive {type=MovieSceneScriptingFloatChannel, keys [[0,1],[4,0]]}
          bOnHover_RGBA {type=MovieSceneScriptingFloatChannel, const [0,1]}
         ]
        }]
      }]
    }
   ]}
 }
 DeselectedPulse {
  binding_count=2
  length_seconds=0.48
  object=&self:DeselectedPulse
  sequence {bindings [
    Image_Icon {
     id=#12
     possessed_class=Image
     tracks [{
       type=MovieSceneWidgetMaterialTrack
       sections [{
         type=MovieSceneParameterSection
         channels [
          bIsInactive {type=MovieSceneScriptingFloatChannel, keys [[0,0],[4,1]]}
          bOnHover_RGBA {type=MovieSceneScriptingFloatChannel, const [0,1]}
         ]
        }]
      }]
    }
    Image_Pulse {
     id=#13
     possessed_class=Image
     tracks [
      {
       type=MovieSceneFloatTrack
       display_name=Render Opacity
       sections [{type=MovieSceneFloatSection, channels [None {type=MovieSceneScriptingFloatChannel, keys [[1,0],[4,1],[5.6,1],[7.6,0],[9.6,0]]}]}]
      }
      {
       type=MovieScene2DTransformTrack
       sections [{
         type=MovieScene2DTransformSection
         channels [
          Scale.X {type=MovieSceneScriptingFloatChannel, keys [[2,1],[7.6,2.5]]}
          Scale.Y {type=MovieSceneScriptingFloatChannel, keys [[2,1],[7.6,2.5]]}
         ]
        }]
      }
     ]
    }
   ]}
 }
]
graphs [
 EventGraph {nodes [
   Event PreConstruct()
   > SetRenderOpacity(self=Get Image_Pulse() -> (Image_Pulse:Image Object Reference), InOpacity=0)
   > InitialUpdateVisual()
   > n1 = SetIsToggleable!(execute, self:Common Button Base Object Reference, bInIsToggleable:bool=true) -> (then)
   SetText!(execute=[@n1.then, @n2.then], self:Self Object Reference, Text:Text=Get Text!() -> (Text)) -> (then)
   > SetIconMaterial!(execute, self, Icon:Material Instance Object Reference=Get Icon() -> (Icon:Material Instance Object Reference)) -> (then)
   > SetIconSize!(execute, self, IconSize:Vector 2D Structure=Get Icon Size() -> (Icon Size:Vector 2D Structure)) -> (then)
   > SetPulseMaterial!(execute, self, Pulse:Material Instance Object Reference=Get PulseIcon() -> (PulseIcon:Material Instance Object Reference)) -> (then)
   > SetIconInverted!(execute, self, IsIconsInverted:bool=Get IsIconsInverted() -> (IsIconsInverted:bool)) -> (then)
   n2 = Event Construct()
   Event On Selected()
   > n3 = Branch(Condition=Get IsInitiallySelected() -> (IsInitiallySelected:bool))
   > Set IsInitiallySelected(IsInitiallySelected:bool=false) -> (Output_Get:bool)
   n4 = Branch(execute=@n3.else, Condition=IsHovered())
   > PlayAnimationForward(InAnimation=Get HoveredSelecting() -> (HoveredSelecting:Widget Animation Object Reference))
   > PlayAnimationForward(InAnimation=Get SelectedPulse() -> (SelectedPulse:Widget Animation Object Reference))
   StopUnfinishedAnimation!(
    execute=@n4.else
    Target:User Widget Object Reference=Self-Reference()
    Animation:Widget Animation Object Reference=Get DeselectedPulse() -> (DeselectedPulse:Widget Animation Object Reference)
   ) -> (then)
   > PlayAnimationForward(InAnimation=Get SelectedPulse())
   Event On Hovered()
   n5 = Select(
    Option 0:Widget Animation Object Reference=Get HoveringDeselected() -> (HoveringDeselected:Widget Animation Object Reference)
    Option 1:Widget Animation Object Reference=Get HoveringSelected() -> (HoveringSelected:Widget Animation Object Reference)
    Index=GetSelected!(self:Common Button Base Object Reference) -> (ReturnValue:bool)
   ) -> (ReturnValue:Widget Animation Object Reference)
   > PlayAnimationForward(InAnimation=@n5)
   Event On Unhovered()
   > PlayAnimationReverse(InAnimation=@n5)
   Event On Deselected()
   > n6 = Branch(Condition=IsHovered())
   > PlayAnimationReverse(InAnimation=Get HoveredSelecting())
   StopUnfinishedAnimation(execute=@n6.else, Target=Self-Reference(), Animation=Get SelectedPulse())
   > PlayAnimationForward(InAnimation=Get DeselectedPulse())
   Event On Disabled() {is=title K2Node_Event}
   > SetColorAndOpacity!(
    execute
    self=Get ButtonText() -> (ButtonText:Common Text Object Reference)
    InColorAndOpacity_SpecifiedColor:Linear Color Structure=(R=1,G=1,B=1,A=0.5)
    InColorAndOpacity_ColorUseRule:ESlateColorStylingMode Enum
   ) -> (then)
   n7 = Get Image_Icon() -> (Image_Icon:Image Object Reference)
   > SetRenderOpacity(self=@n7, InOpacity=0.25)
   Event On Enabled() {is=title K2Node_Event}
   > SetColorAndOpacity(self=Get ButtonText(), InColorAndOpacity_SpecifiedColor=(R=1,G=1,B=1,A=1))
   > SetRenderOpacity(self=@n7, InOpacity=1)
  ]}
 InitialUpdateVisual {nodes [
   InitialUpdateVisual() {is=title K2Node_FunctionEntry}
   > n1 = GetDynamicMaterial(self=Get Background() -> (Background:Image Object Reference))
   > SetScalarParameterValue(self=@n1, ParameterName=GradientFill, Value=0)
   > SetVectorParameterValue(self=@n1, ParameterName=RGB_SolidFill, Value=(R=0,G=0,B=0,A=0.25))
  ]}
 SetIconInverted {nodes [
   n1 = SetIconInverted!() -> (then, IsIconsInverted:bool=~) {is=title K2Node_FunctionEntry}
   n2 = Select(Option 0:float=0, Option 1:float=180, Index=@n1) -> (ReturnValue:float)
   > Set!(
    execute
    RenderTransform_Translation:Vector 2D Structure=(X=0,Y=0)
    RenderTransform_Scale:Vector 2D Structure=(X=1,Y=1)
    RenderTransform_Shear:Vector 2D Structure=(X=0,Y=0)
    RenderTransform_Angle:float=@n2
    self:Widget Object Reference=Get Image_Icon()
   ) -> (then, Output_Get:Widget Transform Structure)
   > Set(
    RenderTransform_Translation=(X=0,Y=0)
    RenderTransform_Scale=(X=1,Y=1)
    RenderTransform_Shear=(X=0,Y=0)
    RenderTransform_Angle=@n2
    self=Get Image_Pulse()
   )
  ]}
 SetIconMaterial {nodes [
   n1 = SetIconMaterial!() -> (then, Icon:Material Instance Object Reference) {is=title K2Node_FunctionEntry}
   > SetBrushFromMaterial(self=Get Image_Icon(), Material=@n1)
  ]}
 SetIconSize {nodes [
   n1 = SetIconSize!() -> (then, IconSize:Vector 2D Structure) {is=title K2Node_FunctionEntry}
   > SetDesiredSizeOverride!(execute, self:Image Object Reference=Get Image_Icon(), DesiredSize:Vector 2D Structure=@n1) -> (then)
   n2 = BreakVector2D(InVec=@n1)
   > SetDesiredSizeOverride(self=Get Image_Pulse(), DesiredSize=MakeVector2D(X=float - float(A=@n2, B=4), Y=float - float(A=@n2.Y, B=4)))
  ]}
 SetIsFavorite {nodes [
   n1 = SetIsFavorite!() -> (then, IsFavorite:bool=~) {is=title K2Node_FunctionEntry}
   > n2 = Set IsInitiallySelected(IsInitiallySelected=@n1)
   > n3 = GetDynamicMaterial(self=Get Image_Icon())
   > SetScalarParameterValue(self=@n3, ParameterName=bIsInactive, Value=Select(Option 0=1, Option 1=0, Index=@n2))
  ]}
 SetPulseMaterial {nodes [
   n1 = SetPulseMaterial!() -> (then, Pulse:Material Instance Object Reference) {is=title K2Node_FunctionEntry}
   > SetBrushFromMaterial(self=Get Image_Pulse(), Material=@n1)
  ]}
 SetText {nodes [
   n1 = SetText!() -> (then, Text:Text) {is=title K2Node_FunctionEntry}
   > "SetText (Text)"(self=Get ButtonText(), InText=@n1)
  ]}
]
variables [
 Icon {type=MaterialInstance, default=/FortUILibrary/Foundation/Buttons/Materials/Favorite/MI_UI_Favorite_ExpressYourSupport}
 Icon Size {type=Vector2D, default=<40 40>}
 IsIconsInverted {type=boolean}
 IsInitiallySelected {type=boolean}
 PulseIcon {type=MaterialInstance, default=/FortUILibrary/Foundation/Buttons/Materials/Favorite/MI_UI_Favorite_Pulse_Modal}
 Text {type=string, default=Favorite}
]
widgets {
 info {inheritedWidgetCount=1, parentClass=&CommonButtonBase, rootWidgetClass=&GridPanel, widgetCount=7}
 widgets [
  GridPanel_28 {widgetClassPath=&GridPanel, widget=GridPanel_28, properties {columnFill [1], rowFill [1]}}
  Background {
   widgetClassPath=&Image
   bIsVariable=true
   parent=GridPanel_28
   widget=Background
   properties {brush {
     bIsDynamicallyLoaded=false
     drawAs=Image
     imageType=FullColor
     mirroring=NoMirror
     resourceName=None
     resourceObject=/Game/UI/Sidebar/Materials/M_UI_sidebar_Button_Selection_Inst
     tiling=NoTile
     imageSize=<32 32>
     margin=<ltrb 0 0 0 0>
     outlineSettings {
      bUseBrushTransparency=false
      roundingType=HalfHeightRadius
      width=0
      color {colorUseRule=UseColor_Specified, specifiedColor=<rgba 0 0 0 0>}
      cornerRadii=<0 0 0 0>
     }
     tintColor {colorUseRule=UseColor_Specified, specifiedColor=<rgba 1 1 1 1>}
     uVRegion {bIsValid=false, max=<0 0>, min=<0 0>}
    }}
   slot {class=GridSlot}
  }
  InputActionWidget {
   widgetClassPath=&CommonActionWidget
   bInherited=true
   bIsVariable=true
   parent=GridPanel_28
   widget=InputActionWidget
   properties {
    visibility=SelfHitTestInvisible
    iconRimBrush {
     bIsDynamicallyLoaded=false
     drawAs=Image
     imageType=FullColor
     mirroring=NoMirror
     resourceName=None
     resourceObject=/Game/UI/Sidebar/Materials/MI_SidebarRoundedRectangleButton
     tiling=NoTile
     imageSize=<32 32>
     margin=<ltrb 0 0 0 0>
     outlineSettings {
      bUseBrushTransparency=false
      roundingType=HalfHeightRadius
      width=0
      color {colorUseRule=UseColor_Specified, specifiedColor=<rgba 0 0 0 0>}
      cornerRadii=<0 0 0 0>
     }
     tintColor {colorUseRule=UseColor_Specified, specifiedColor=<rgba 1 1 1 1>}
     uVRegion {bIsValid=false, max=<0 0>, min=<0 0>}
    }
   }
   slot {class=GridSlot, properties {horizontalAlignment=HAlign_Right, layer=1, verticalAlignment=VAlign_Center, padding=<ltrb 0 0 10 0>}}
  }
  GridPanel_73 {
   widgetClassPath=&GridPanel
   parent=GridPanel_28
   widget=GridPanel_73
   properties {columnFill [1], rowFill [0,1]}
   slot {class=GridSlot, properties {horizontalAlignment=HAlign_Center, layer=1, verticalAlignment=VAlign_Bottom, padding=<ltrb 0 0 0 22>}}
  }
  Image_Pulse {
   widgetClassPath=&Image
   bIsVariable=true
   parent=GridPanel_73
   widget=Image_Pulse
   properties {visibility=SelfHitTestInvisible}
   slot {class=GridSlot, properties {horizontalAlignment=HAlign_Center, verticalAlignment=VAlign_Top}}
  }
  Image_Icon {
   widgetClassPath=&Image
   bIsVariable=true
   parent=GridPanel_73
   widget=Image_Icon
   properties {
    visibility=SelfHitTestInvisible
    brush {
     bIsDynamicallyLoaded=false
     drawAs=Image
     imageType=FullColor
     mirroring=NoMirror
     resourceName=None
     resourceObject=/FortUILibrary/Foundation/Buttons/Materials/Favorite/MI_UI_ThumbsUp_ExpressYourSupport
     tiling=NoTile
     imageSize=<32 32>
     margin=<ltrb 0 0 0 0>
     outlineSettings {
      bUseBrushTransparency=false
      roundingType=HalfHeightRadius
      width=0
      color {colorUseRule=UseColor_Specified, specifiedColor=<rgba 0 0 0 0>}
      cornerRadii=<0 0 0 0>
     }
     tintColor {colorUseRule=UseColor_Specified, specifiedColor=<rgba 1 1 1 1>}
     uVRegion {bIsValid=false, max=<0 0>, min=<0 0>}
    }
   }
   slot {class=GridSlot, properties {horizontalAlignment=HAlign_Center, verticalAlignment=VAlign_Top}}
  }
  ButtonText {
   widgetClassPath=&CommonTextBlock
   bIsVariable=true
   parent=GridPanel_73
   widget=ButtonText
   properties {
    mobileFontSizeMultiplier=1.3
    scrollStyle=&TextScrollStyle-Base_C
    text=Favorite
    textTransformPolicy=ToUpper
    font {
     bForceMonospaced=false
     bMaterialIsStencil=false
     fontMaterial=null
     fontObject=&BurbankSmall
     letterSpacing=0
     monospacedWidth=1
     size=13.5
     skewAmount=0
     typefaceFontName=Black
     outlineSettings {
      bApplyOutlineToDropShadows=false
      bMiteredCorners=false
      bSeparateFillAlpha=false
      outlineMaterial=null
      outlineSize=0
      outlineColor=<rgba 0 0 0 1>
     }
    }
    shadowOffset=<0 0>
   }
   slot {class=GridSlot, properties {row=1, verticalAlignment=VAlign_Bottom, padding=<ltrb 0 9 0 0>}}
  }
 ]
}
