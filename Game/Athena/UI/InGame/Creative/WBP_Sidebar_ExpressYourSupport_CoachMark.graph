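# Sidebar onboarding coach mark that asks players to give feedback on islands they play, built on FortSidebarOnboardTooltipWidget.
# It wraps a modal WBP_UIKit_CoachMark with a single Escape-bound Close button as focus target, styled in PreConstruct along with the description color.
# OnShowTooltip activates the coach mark.
# Close clicks (after HandleOtherActionClicked), OnEndTooltip and focus-path loss all call NotifyCloseAnimationFinished directly, then deactivate it.
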WidgetBlueprint /Game/Athena/UI/InGame/Creative/WBP_Sidebar_ExpressYourSupport_CoachMark
graphs [
 BP_GetDesiredFocusTarget {nodes [
   Get Desired Focus Target()
   > Return Node!(
    execute
    ReturnValue:Widget Object Reference=Get OtherActionButton!(self:WBP UIKit Coach Mark Base Object Reference=Get WBP_UIKit_CoachMark() -> (WBP_UIKit_CoachMark:WBP UIKit Coach Mark Object Reference)) -> (OtherActionButton:WBP UIKit Button Regular Object Reference)
   )
  ]}
 EventGraph {nodes [
   "Button Other Action Clicked (WBP_UIKit_CoachMark)"() {is=title K2Node_ComponentBoundEvent}
   > n1 = HandleOtherActionClicked(self:Fort Sidebar Onboard Tooltip Widget Object Reference)
   NotifyCloseAnimationFinished(execute=[@n1.then, @n2.then, @n3.then], self:Fort Sidebar Onboard Tooltip Widget Object Reference)
   > DeactivateWidget(self=Get WBP_UIKit_CoachMark())
   n2 = Event OnEndTooltip() {is=title K2Node_Event}
   Event PreConstruct()
   > SetButtonType!(
    execute
    self:WBP UIKit Button Regular Object Reference=Get OtherActionButton(self=Get WBP_UIKit_CoachMark())
    ButtonType:E_UI_CTAButtonType Enum=NewEnumerator1
   ) -> (then)
   > SetDefaultColorAndOpacity!(
    execute
    self:Rich Text Block Object Reference=Get RichText_Description!(self:WBP UIKit Coach Mark Base Object Reference=Get WBP_UIKit_CoachMark()) -> (RichText_Description:Common Rich Text Block Object Reference)
    InColorAndOpacity_SpecifiedColor:Linear Color Structure=(R=0.5776,G=0.5776,B=0.5776,A=1)
    InColorAndOpacity_ColorUseRule:ESlateColorStylingMode Enum
   ) -> (then)
   n3 = Event OnRemovedFromFocusPath!() -> (OutputDelegate, then, InFocusEvent:Focus Event Structure) {is=title K2Node_Event}
   Event OnShowTooltip() {is=title K2Node_Event}
   > ActivateWidget(self=Get WBP_UIKit_CoachMark())
  ]}
]
variables []
widgets {
 info {
  inheritedWidgetCount=2
  parentClass=/Script/FortniteUI.FortSidebarOnboardTooltipWidget
  rootWidgetClass=/FortUILibrary/UIKit/CoachMark/WBP_UIKit_CoachMark.WBP_UIKit_CoachMark_C
  widgetCount=3
 }
 widgets [
  WBP_UIKit_CoachMark {
   widgetClassPath=/FortUILibrary/UIKit/CoachMark/WBP_UIKit_CoachMark
   bIsVariable=true
   widget=WBP_UIKit_CoachMark
   properties {
    arrowPosition=Right_Middle
    bIsFocusable=true
    bIsModal=true
    navigation=&self:WidgetTree.WBP_UIKit_CoachMark.WidgetNavigation_0
    shouldShowDismissButton=false
    textDescription="Let us know how you feel about the islands you play! Your feedback will help us improve your Discover experience."
    textHeader=Before You Go...
    textOtherActionButton=Close
    visibility=SelfHitTestInvisible
    input Action Dismiss {dataTable=&InputActionDataTable, rowName=GenericEscape}
    input Action Other {dataTable=&InputActionDataTable, rowName=GenericEscape}
   }
  }
  Dismiss_Button {widgetClassPath=&CommonButtonLegacy, bInherited=true, widget=null}
  OtherAction_Button {widgetClassPath=&CommonButtonLegacy, bInherited=true, widget=null}
 ]
}
